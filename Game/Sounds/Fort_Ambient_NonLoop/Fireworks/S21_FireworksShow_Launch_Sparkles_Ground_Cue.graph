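# Sound cue for the ground sparkles launch sound in the S21 fireworks show, played in the world through the live in-world sound class.
# A single wave player feeds one sparkle wave straight to the output at reduced volume, with no randomization or mixing.
# Spatialization and very long-range falloff come from the shared fireworks show attenuation asset, while a ground fireworks concurrency set governs simultaneous playback.
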SoundCue /Game/Sounds/Fort_Ambient_NonLoop/Fireworks/S21_FireworksShow_Launch_Sparkles_Ground_Cue
root=WavePlayer_4
nodes [{id=WavePlayer_4, properties {soundWaveAssetPtr=&/S21_FireworksShow_Launch_Sparkles_Ground_01}}]
properties {
 attenuationSettings=&/S21_FireworksShow_Att
 maxDistance=180000
 soundClassObject=/Game/Sounds/FortSoundClassesAndModes/Fort_Live_InWorld
 volumeMultiplier=0.2
 attenuationOverrides {
  bEnableFocusInterpolation=true
  bEnableListenerFocus=true
  falloffDistance=75000
  focusAzimuth=45
  nonFocusAzimuth=90
  nonFocusVolumeAttenuation=0.5
  nonSpatializedRadiusStart=10000
  reverbDistanceMax=40000
  reverbDistanceMin=4000
  reverbWetLevelMax=0.4
  reverbWetLevelMin=0.1
  stereoSpread=50000
  attenuationShapeExtents {x=20000}
 }
 concurrencyOverrides {maxCount=1}
 concurrencySet [&/Fireworks_Conc_Ground]
}
